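# Translucent lit glass material for the jar on the SwampFish backpack, with fresnel-driven tint and opacity.
# Tint, base texture and purple spots multiply scene color sampled at a normal-, fresnel- and depth-scaled screen offset, faking refraction.
# A cubemap reflection, with mip level varied by tiled grime, is added before a fresnel color multiply.
# Opacity switches by quality tier, mobile color drops to a flat tint, and a proxy replace skips scene color and reflection.
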Material /Game/Accessories/FORT_Backpacks/Backpack_M_MED_SwampFish/FX/Materials/M_Backpack_SwampFish_GlassJar
graph {
 properties {
  bScreenSpaceReflections=true
  bUseMaterialAttributes=true
  bUsedWithClothing=true
  bUsedWithMeshDeformer=true
  bUsedWithNanite=true
  bUsedWithNiagaraMeshParticles=true
  bUsedWithSkeletalMesh=true
  blendMode=BLEND_Translucent
  pixelDepthOffsetMode=PDOM_Legacy
  translucencyLightingMode=TLM_VolumetricDirectional
  translucencyPass=MTP_BeforeDOF
  usedShadingModels=Default Lit
 }
 nodes [
  n1 = VectorParameter() {parameterName=Glass Tint Center, defaultValue=<rgba 0.1719 1 0.4755 1>}
  n2 = VectorParameter() {parameterName=Glass Tint Edge, defaultValue=<rgba 0 1 0.9 1>}
  n3 = DotProduct(A=CameraVectorWS(), B=PixelNormalWS())
  n4 = OneMinus(None=Power(Base=Abs(None=@n3), Exp=ScalarParameter() {defaultValue=2, parameterName=Fresnel Exponent}))
  n5 = LinearInterpolate(
   A=LinearInterpolate(
    A=LinearInterpolate(A=@n1.RGB, B=@n2.RGB, Alpha=@n4)
    B=TextureSampleParameter2D() {parameterName=BaseColor Texture, texture=/Game/Accessories/FORT_Backpacks/Backpack_M_MED_SwampFish/Textures/T_SwampFish_Backpack_D}.RGB
    Alpha=ScalarParameter() {defaultValue=1, parameterName=Base Color Texture Lerp}
   )
   B=VectorParameter() {parameterName=Purple Spots, defaultValue=<rgba 0.659 0.4 1 1>}.RGB
   Alpha=Multiply(A=NamedRerouteUsage(), B=ScalarParameter() {defaultValue=0.6, parameterName=Purple Spots Lerp})
  )
  n6 = Divide(A=ScalarParameter() {defaultValue=-1, parameterName=BlurAmount, sortPriority=19}, B=PixelDepth().R)
  n7 = PixelNormalWS()
  n8 = ComponentMask(None=Multiply(A=Multiply(A=@n4, B=@n6), B=Transform(None=@n7) {transformSourceType=TRANSFORMSOURCE_World, transformType=TRANSFORM_View})) {g=true, r=true}
  n9 = Constant2Vector() {g=-1, r=1}
  n10 = Multiply(A=@n5, B=SceneColor!(OffsetFraction=Multiply(A=@n8, B=@n9)) {inputMode=OffsetFraction}.RGB)
  n11 = TextureSampleParameter2D(UVs=Multiply(A=TextureCoordinate(), B=ScalarParameter() {defaultValue=5, parameterName=Grime Texture Tiling})) {parameterName=GrimeTexture, samplerType=SAMPLERTYPE_Masks, texture=/Game/Global/Textures/Default/Noise/T_Tiling_Grime}
  n12 = TextureSampleParameterCube!(
   UVs=ReflectionVectorWS()
   Level=LinearInterpolate(
    A=ScalarParameter() {parameterName=Cubemap Miplevel Min}
    B=ScalarParameter() {defaultValue=6, parameterName=cubemap Miplevel Max}
    Alpha=NamedRerouteDeclaration(None=@n11.R) {name=GrimeTexture, nodeColor=<rgba 0.3125 0 1 1>}
   )
   Apply View MipBias
  ) {
   mipValueMode=TMVM_MipLevel
   parameterName=Cube Map
   samplerType=SAMPLERTYPE_LinearColor
   texture=/Game/Creative/Environments/Meshes/Glass/Textures/T_Glass_Cubemap_01_Desaturated
  }
  n13 = Multiply(
   A=@n12.RGB
   B=LinearInterpolate(
    A=ScalarParameter() {defaultValue=0.333, parameterName=Cube Map influence Center}
    B=ScalarParameter() {defaultValue=1, parameterName=Cube Map influence Edges}
    Alpha=@n4
   )
  )
  n14 = VectorParameter() {parameterName=Fresnel_Color, defaultValue=<rgba 0.2828 0.724 0.6872 1>}
  n15 = MaterialFunctionCall(
   Brightness=ScalarParameter() {defaultValue=2, parameterName=Fresnel_Brightness}
   Exponent=ScalarParameter() {defaultValue=0.5, parameterName=Fresnel_Exponent}
  ) {materialFunction=&MF_Fresnel_FX}
  n16 = Multiply(
   A=MaterialProxyReplace(Realtime=Add(A=@n10, B=@n13), MaterialProxy=@n5)
   B=LinearInterpolate(B=@n14.RGB, Alpha=@n15.Fresnel Inverted) {constA=0.5}
  )
  n17 = ScalarParameter() {defaultValue=0.6, parameterName=Fresnel Inner Transparency}
  n18 = ScalarParameter() {defaultValue=0.85, parameterName=Fresnel Outer Transparency}
  n19 = Add(
   A=LinearInterpolate(
    A=LinearInterpolate(A=@n17, B=@n18, Alpha=@n4)
    B=ScalarParameter() {defaultValue=0.8, parameterName=Center Opacity}
    Alpha=NamedRerouteUsage()
   )
   B=Saturate(None=@n15.Fresnel Inverted)
  )
  n20 = ScalarParameter() {defaultValue=0.21, parameterName=OpacityLowEndMulti}
  n21 = LinearInterpolate(
   A=ScalarParameter() {defaultValue=0.4, parameterName=Mobile_Opacity0}
   B=ScalarParameter() {defaultValue=1, parameterName=Mobile_Opacity1}
   Alpha=NamedRerouteUsage()
  ) {constA=0.8, constB=0.5}
  n22 = Multiply(A=@n21, B=@n20) {constB=0.55}
  n23 = Clamp(None=@n19) {maxDefault=0.7}
  n24 = TextureSampleParameter2D() {
   group=Mask Texture
   parameterName=BakedNormal
   samplerSource=SSM_Wrap_WorldGroupSettings
   samplerType=SAMPLERTYPE_Normal
   texture=/Game/Accessories/FORT_Backpacks/Backpack_M_MED_SwampFish/Textures/T_SwampFish_Backpack_N
  }
  MakeMaterialAttributes!(
   BaseColor=ShadingPathSwitch(Default=@n16, Deferred=@n16, Mobile=Multiply(A=@n14.RGB) {constB=0.8})
   Metallic=ScalarParameter() {parameterName=Metallic}
   Specular=ScalarParameter() {defaultValue=1, parameterName=Specular}
   Roughness=ScalarParameter() {defaultValue=0.15, parameterName=Roughness}
   Anisotropy
   EmissiveColor
   Opacity=MaterialFunctionCall!(
    Low _ PC Low and Med=Multiply(A=@n19, B=@n20) {constB=0.55}
    "High _ XB1, PS4"=@n23
    "Epic _ PC High, XSX, PS5"=@n23
    --------
    "Mobile Low _ PC Perf, Android"=@n22
    Mobile Med _ Switch=@n22
   ) {materialFunction=/Game/Global/MaterialFunctions/TechArt/MF_Fortnite_QualitySwitch_Scalar}.Result
   OpacityMask
   Normal=Normalize(VectorInput=@n24.RGB)
   Tangent
   WorldPositionOffset
   SubsurfaceColor
   ClearCoat
   ClearCoatRoughness
   AmbientOcclusion
   Refraction
   CustomizedUVs_0
   CustomizedUVs_1
   CustomizedUVs_2
   CustomizedUVs_3
   CustomizedUVs_4
   CustomizedUVs_5
   CustomizedUVs_6
   CustomizedUVs_7
   PixelDepthOffset
   ShadingModel
   Displacement
  )
  n25 = ScalarParameter() {defaultValue=5, parameterName=Mobile_FresnelExponent}
  n26 = ScalarParameter() {defaultValue=0.04, parameterName=Mobile_FresnelBaseReflectFraction}
  NamedRerouteDeclaration(None=LinearInterpolate(
    A=NamedRerouteUsage()
    B=Fresnel(ExponentIn=@n25, BaseReflectFractionIn=@n26)
    Alpha=ScalarParameter() {defaultValue=1, parameterName=Mobile_ReplaceGrimWithFresnel}
   )) {name=Mobile Opacity, nodeColor=<rgba 0.3828 0 1 1>}
  n27 = ShadingPathSwitch(Default=@n19, Deferred=@n19, Mobile=@n21)
  QualitySwitch(Default=@n27, Low=Multiply(A=@n27) {constB=0.55}, High=@n27, Medium=@n27, Epic=@n27)
 ]
}
